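; Parent Blueprint for Creative Trigger devices, providing shared activation logic that child triggers extend.
; On authority, activation passes use-restriction, enabled, reset and trigger-count checks, optionally waits on a delay, then fires OnTrigger every Nth activation.
; Minigame phases and receiver events control enabling and counters, while a CurrentState RepNotify selects on, delay or off materials.
; Throttled hit cosmetics play async-loaded audio, VFX and a timeline glow on cache-borrowed dynamic materials.
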(event OnVisibilityChanged(VisibleInGameComponent)
  (params (Visible :type Boolean))
  ;[1 1 1] Visibility
  (CallFunction|SetMeshVisibilityandCollision Visible))

(event Custom|UpdateVisibilityDuringPhase
  ;[1 1 1] Visibility
  (Networking|SwitchHasAuthority
    (:Authority
      (Visibility|SetVisiblityDuringPhase (Variables|Default|GetVisibleInGameComponent) (select (Variables|Options|GetVisibilityInGame) "NewEnumerator3" "NewEnumerator1")))))

(event EventBeginPlay
  ;[0 1 0.3633 0.5] Initialization
  (Visibility|SetVisibilityComponents (Variables|Default|GetVisibleInGameComponent) (Utilities|Array|MakeArray (Variables|Default|GetMesh) (Variables|Default|GetChannelNumberPlane) (Variables|Default|GetP_Trigger_Activate)) true))

(event Custom|Load%20and%20Set%20Audio
  ;[1 0.7441 0] Async Load and set the various audio samples for the Trigger
  (Networking|CanExecuteCosmeticEvents
    (:True
      (if (Utilities|UseTriggerSound)
        (for _array_element (Utilities|GetTriggerSounds)
          (bind _object (Utilities|AsyncLoadAsset _array_element)
            (:Completed
              (bind _assound_cue (Utilities|Casting|CastToSoundCue _object))
              (if (Utilities|Equal(Object) (Utilities|ResolveSoftReference (Utilities|Array|Get(acopy) (Variables|SFX|GetTriggerSounds) (Variables|Options|GetTriggerSoundIndex))) _assound_cue)
                (Audio|Components|Audio|SetSound (Variables|Default|GetActivate%20Audio) _assound_cue)))))
        (bind _object_2 (Utilities|AsyncLoadAsset (Variables|SFX|GetResetSound))
          (:Completed
            (bind _assound_cue_2 (Utilities|Casting|CastToSoundCue _object_2))
            (Audio|Components|Audio|SetSound (Variables|Default|GetReset%20Audio) _assound_cue_2)))
        (else
          (Audio|Components|Audio|SetSound (Variables|Default|GetActivate%20Audio) 0)
          (Audio|Components|Audio|SetSound (Variables|Default|GetReset%20Audio) 0))))))

(event Custom|Initialize%20Device
  ;[0 1 0.3633 0.5] Initialization
  (CallFunction|InitializeComponents)
  (CallFunction|EnableOnFirstValidPhase "PreGame") ; Calling here as we may be in pre-game in a featured island or LTM
  (CallFunction|LoadAndSetAudio))

(event OnAnyPropertyChanged(ToyOptionsComponent)
  ;[0 1 0.3633 0.5] Initialization
  (CallFunction|InitializeComponents)
  (CallFunction|EnableOnFirstValidPhase "PreGame") ; Calling here as we may be in pre-game in a featured island or LTM
  (CallFunction|LoadAndSetAudio))

(event OnPropertiesLoaded(ToyOptionsComponent)
  ;[0 1 0.3633 0.5] Initialization
  (CallFunction|InitializeComponents)
  (CallFunction|EnableOnFirstValidPhase "PreGame") ; Calling here as we may be in pre-game in a featured island or LTM
  (CallFunction|LoadAndSetAudio))

(event EventEndPlay
  (params (EndPlayReason :type EEndPlayReason%20Enum :default "Destroyed"))
  ;[0 1 0.3633 0.5] Initialization
  (CallFunction|SwitchToOffMaterial))

(event Custom|Hit
  (params (Play%20VFX :type Boolean))
  ;[1 0.8839 0] Cosmetic Effect on Clients
  (Networking|CanExecuteCosmeticEvents
    (:True
      (CallFunction|SwitchToDyamicMaterial)
      (CallFunction|AnimateDynamicMaterialActivate)
      (Audio|Components|Audio|Play (Variables|Default|GetActivate%20Audio) 0.0)
      (if (and Play%20VFX (Variables|Default|GetTriggerVFX))
        (Components|Activation|SetActive (Variables|Default|GetP_Trigger_Activate) true true)))))

(event Custom|Send%20Hit
  (params (Play%20FX :type Boolean))
  ;[1 0.6759 0] Only send to Client Every MinFXDelay seconds
  (Utilities|FlowControl|DoOnce false
    (:Completed
      (CallFunction|Hit Play%20FX)
      (Utilities|FlowControl|Delay (Variables|Default|GetMin%20FX%20Delay)))))

(event Custom|Minigame%20State%20Changed
  (params (New%20State :type EFortMinigameState%20Enum :default "PreGame"))
  ;[0 0.08007 1 0.5] Gameplay State Logic
  (CallFunction|EnableOnFirstValidPhase New%20State))

(event MinigameStateChangedDelegate(FortMinigameLogic)
  (params (Minigame :type Fort%20Minigame%20Object%20Reference) (NewMinigameState :type EFortMinigameState%20Enum :default "PreGame"))
  ;[0 0.08007 1 0.5] Gameplay State Logic
  (CallFunction|MinigameStateChanged NewMinigameState))

(event MinigameEndedDelegate(FortMinigameLogic)
  ;[0 0.08007 1 0.5] Gameplay State Logic
  (Utilities|Time|ClearandInvalidateTimerbyHandle (Variables|Default|GetTriggerDelayTimer))
  (Utilities|Time|ClearandInvalidateTimerbyHandle (Variables|Default|GetResetTimer))
  (CallFunction|Disable))

(event Custom|EndReset
  ;[0 0.432 1] Ending the Reset Period
  (Variables|Default|SetbIsResetting false)
  (if (CallFunction|CanTrigger)
    (Variables|Default|SetCurrentState "NewEnumerator0")
    (else
      (Variables|Default|SetCurrentState "NewEnumerator2"))))

(event Custom|SendActivationEvent
  ;[0 0.3961 1] Actually Activating the Device.
  (if (and (> (Variables|Options|GetTimesCanTrigger) 0) (not (Variables|Default|GetbChildHandlesDecrement)))
    (Variables|Default|SetRemainingTimesCanTrigger (- (Variables|Default|GetRemainingTimesCanTrigger) 1)))
  (CallFunction|AttemptTrigger)
  (if (> (Variables|Options|GetTriggerResetDelay) 0.0)
    (bind _returnvalue (Utilities|Time|SetTimerbyEvent (EventDispatchers|CreateEvent :!event Custom|EndReset) 0.01 false false (Variables|Options|GetTriggerResetDelay)))
    (Variables|Default|SetResetTimer _returnvalue)
    (Variables|Default|SetCurrentState "NewEnumerator2")
    (else
      (CallFunction|EndReset))))

(event Custom|EnableFromEvent
  (params (Instigator :type Controller%20Object%20Reference))
  ;[0.6648 0 1] Receivers
  ;;[1 1 1] Activate/Deactivate when Received goes here
  (CallFunction|Enable) ; Call function to give children something\nto hook into. for custom logic here.
)

(event Custom|Enable
  ;[0.6648 0 1] Receivers
  (Variables|Default|SetbIsEnabled true)
  ;;[1 1 1] Don't interrupt the current trigger if we are triggering
  (if (Variables|Default|GetbIsResetting)
    _
    (elif (CallFunction|CanTrigger)
      (Variables|Default|SetCurrentState "NewEnumerator0"))))

(event Custom|Disable
  ;[0.6648 0 1] Receivers
  (Variables|Default|SetbIsEnabled false)
  ;;[1 1 1] Don't interrupt the current trigger if we are triggering
  (if (Variables|Default|GetbIsResetting)
    _
    (elif (CallFunction|CanTrigger)
      _
      (else
        (Variables|Default|SetCurrentState "NewEnumerator2")))))

(event Custom|DisableFromEvent
  (params (Instigator :type Controller%20Object%20Reference))
  ;[0.6648 0 1] Receivers
  ;;[1 1 1] Activate/Deactivate when Received goes here
  (CallFunction|Disable) ; Call function to give children something\nto hook into. for custom logic here.
)

(event Custom|Switch%20to%20Dyamic%20Material
  (bind _material_instance_dynamic (Variables|Materials|GetMaterial%20Instance%20Dynamic))
  ;[1 0.7873 0] Material Updates
  (Networking|CanExecuteCosmeticEvents
    (:True
      (Variables|Materials|GetMaterial%20Instance%20Dynamic
        (:then
          _)
        (:else
          (bind _material_cache (Variables|Materials|GetMaterial%20Cache)
            (:then
              (bind (_outmaterialinstance _returnvalue) (Class|FortMaterialCache|ReserveMaterial _material_cache))
              (Variables|Materials|SetMaterial%20Instance%20Dynamic _outmaterialinstance)
              (if _returnvalue
                _))
            (:else))))
      (Rendering|Material|SetMaterial (Variables|Default|GetMesh) 1 _material_instance_dynamic)
      (Variables|Materials|SetCurrentMaterial _material_instance_dynamic))))

(event Custom|ResetTimesTriggeredFromEvent
  (params (Instigator :type Controller%20Object%20Reference))
  ;[0.6648 0 1] Receivers
  (CallFunction|ResetTriggerCount)
  (if (CallFunction|CanTrigger)
    (Variables|Default|SetCurrentState "NewEnumerator0")))

(event Custom|Switch%20to%20Animating%20Material
  ;[1 0.7873 0] Material Updates
  (Networking|CanExecuteCosmeticEvents
    (:True
      (CallFunction|SetCurrentConstantMaterial (Variables|Materials|GetDefaultDelayMaterial)))))

(event Custom|ReleaseDynamicMaterial
  (bind _material_cache (Variables|Materials|GetMaterial%20Cache))
  (bind _material_instance_dynamic (Variables|Materials|GetMaterial%20Instance%20Dynamic))
  ;[1 0.7873 0] Material Updates
  (if (and (Utilities|IsValid _material_cache :!class /Script/Engine.KismetSystemLibrary) (Utilities|IsValid _material_instance_dynamic :!class /Script/Engine.KismetSystemLibrary))
    (Class|FortMaterialCache|ReleaseMaterial _material_cache _material_instance_dynamic)))

(event Custom|Switch%20to%20Off%20Material
  ;[1 0.7873 0] Material Updates
  (Networking|CanExecuteCosmeticEvents
    (:True
      (CallFunction|SetCurrentConstantMaterial (Variables|Materials|GetDefaultOffMaterial)))))

(event Custom|Switch%20to%20On%20Material
  ;[1 0.7873 0] Material Updates
  (Networking|CanExecuteCosmeticEvents
    (:True
      (CallFunction|SetCurrentConstantMaterial (Variables|Materials|GetDefaultOnMaterial))
      (Audio|Components|Audio|Play (Variables|Default|GetReset%20Audio) 0.0))))

(event Custom|Animate%20Dynamic%20Material%20Activate
  ;[1 0.7873 0] Material Updates
  (bind _material_instance_dynamic (Variables|Materials|GetMaterial%20Instance%20Dynamic)
    (:then
      (bind (_direction _glow) (Timeline|Triggered PlayFromStart)
        (:Update
          (Rendering|Material|SetScalarParameterValue _material_instance_dynamic "Glow" _glow :!class /Script/Engine.MaterialInstanceDynamic))
        (:Finished
          (Rendering|Material|SetScalarParameterValue _material_instance_dynamic "Glow" _glow :!class /Script/Engine.MaterialInstanceDynamic)
          (Default|SetMaterialBasedonCurrentState))))
    (:else)) ; Skip animating if we failed to get a dynamic\nmaterial instance (too may things animating)
)

(fn ActivateFromOtherDevice
  (params (Instigator :type Controller%20Object%20Reference) (Requester :type ERequesterType%20Enum :default "NewEnumerator0"))
  (return false))

(fn AttemptTrigger
  ;[0 0.7828 1] Attempting Trigger
  (bind _can_trigger (CallFunction|DoNTriggerLogic))
  (CallFunction|SendHit _can_trigger)
  (if _can_trigger
    (CallFunction|TransmitMessage)
    (return)))

(fn BlueprintModifyIncomingDamage
  (params (Damage :type Float%20%28single-precision%29) (InTags :type Gameplay%20Tag%20Container%20Structure%20%28by%20ref%29) (EffectContext :type Gameplay%20Effect%20Context%20Handle%20Structure%20%28by%20ref%29) (EventInstigator :type Controller%20Object%20Reference) (DamageCauser :type Actor%20Object%20Reference))
  (return 0.0))

(fn CanInteract
  (params (InteractingPawn :type Fort%20Pawn%20Object%20Reference) (InteractionBeingAttempted :type EInteractionBeingAttempted%20Enum :default "FirstInteraction") (InteractionType :type TInteractionType%20Enum :default "IT_NoInteraction"))
  (return false))

(fn CanTrigger
  (Utilities|FlowControl|Sequence
    (:then_0
      (LocalVariables|SetbOutCanTrigger (Math|Boolean|ANDBoolean (Class|CreativeCommonDeviceFunctionLibrary|IsinPlayMode self) (not (Variables|Default|GetbIsResetting)) (or (<= (Variables|Options|GetTimesCanTrigger) 0) (> (Variables|Default|GetRemainingTimesCanTrigger) 0)) (Variables|Default|GetbIsEnabled))))
    (:then_1
      (return (LocalVariables|GetbOutCanTrigger)))))

(fn DoNTriggerLogic
  (bind _timestriggered (Variables|Default|GetTimesTriggered))
  (Math|Integer|IncrementInt _timestriggered
    (:%20%20
      (return (== (Math|Integer|%25%28Integer%29 _timestriggered (Variables|Options|GetTransmitEveryXTriggers)) 0)))))

(fn EnableonFirstValidPhase
  (params (Current%20State :type EFortMinigameState%20Enum :default "PreGame"))
  (if (or (and (Utilities|Enum|Equal(Enum) Current%20State "PreGame") (Class|CreativeCommonDeviceFunctionLibrary|IsinPlayMode self)) (Utilities|Enum|Equal(Enum) Current%20State "Transitioning"))
    (CallFunction|ResetTriggerCount)
    (Variables|Default|SetbIsResetting false)
    (if (Variables|Options|GetEnabledOnMinigameStart)
      (CallFunction|Enable)
      (else
        (CallFunction|Disable)))))

(fn GetDeviceColor
  (return 0))

(fn GetInstigatingController
  (return 0))

(fn GetCollisionStaticMesh
  (return "0, 0, 0" (Variables|/Script/Engine.StaticMeshComponent|GetStaticMesh (Variables|Default|GetMesh))))

(fn GetMeshComponents
  (return (Utilities|Array|MakeArray (Variables|Default|GetMesh))))

(fn GetMusicSequencerHitSound
  (params (InstigatedBy :type Controller%20Object%20Reference))
  (return 0 0 0 0))

(fn GetMusicSequencerHitSoundLocation
  (params (InstigatedBy :type Controller%20Object%20Reference))
  (return "0, 0, 0"))

(fn GetOverrideMeshMaterial
  (params (MeshComp :type Static%20Mesh%20Component%20Object%20Reference) (MatIdx :type Integer))
  (bind _mesh (Variables|Default|GetMesh))
  (if (Utilities|Equal(Object) (Variables|GetMeshComp) _mesh)
    (return (Utilities|Select :Index (Variables|GetMatIdx) :Option%200 (Rendering|Material|GetMaterial _mesh 0) :Option%201 (Variables|Materials|GetCurrentMaterial)))
    (elif (Utilities|Equal(Object) (Variables|GetMeshComp) (Variables|Default|GetChannelNumberPlane))
      (if (== (Variables|GetMatIdx) 0)
        (return (Variables|Materials|GetChannel%20Material%20Instance))
        (else
          (return 0)))
      (else
        _))))

(fn GetSpawnedPickups
  (return 0))

(fn GetTeamSettingValue
  (return (Variables|Options|GetTeamSetting)))

(fn IgnoreWorldGridScaleOverride
  (return false))

(fn InitializeComponents
  (CallFunction|UpdateVisibilityDuringPhase)
  (Initialization|InitializeComponent (Variables|Default|GetCreative_UseRestriction_Component) (Variables|Default|GetFortMinigameLogic) 0 0 (CallFunction|GetTeamSettingValue) (Variables|Options|GetClassIndex) 0 1))

(fn IsEnabled
  (return false))

(fn IsHoldingItemForPawn
  (params (Item :type Item%20Variant%20Handle%20Structure) (Pawn :type Pawn%20Object%20Reference))
  (return false))

(fn NotTriggeringSelf
  ;[1 1 1] Very basic check to make sure this isn't triggering itself.
  ; Players can easily get around this by having two triggers trigger each other, but this can be done with other devices as well, If a creator screws over their own experience, that's on them.
  ; Children will define their own implementations.
  (return true))

(fn OnRep_CurrentState
  (if (not (Components|Timeline|IsPlaying (Variables|BP_Creative_Trigger_Parent|GetTriggered)))
    (Default|SetMaterialBasedonCurrentState)))

(fn PhysicalRetriggerEnterLogic
  (params (Triggering%20Object :type Actor%20Object%20Reference) (Triggering%20Component :type Primitive%20Component%20Object%20Reference))
  (if (Utilities|Array|ContainsItem (Variables|Default|GetPhysical%20Actor%20Collision%20Types) (Collision|GetCollisionObjectType Triggering%20Component))
    (if (Utilities|Array|ContainsItem (Variables|Default|GetTriggering%20Physical%20Actors) Triggering%20Object)
      (return false)
      (else
        (Utilities|Array|Add (Variables|Default|GetTriggering%20Physical%20Actors) Triggering%20Object)))
    (else
      _))
  (return true))

(fn PhysicalRetriggerLeaveLogic
  (params (Triggering%20Object :type Actor%20Object%20Reference))
  (Utilities|Array|RemoveItem (Variables|Default|GetTriggering%20Physical%20Actors) Triggering%20Object))

(fn ResetTriggerCount
  (Variables|Default|SetRemainingTimesCanTrigger (Variables|Options|GetTimesCanTrigger))
  (Variables|Default|SetTimesTriggered 0))

(fn SetTriggerCollision
  (params (On :type Boolean))
  (return false))

(fn SetCurrentConstantMaterial
  (params (NewCurrentMaterial :type Material%20Interface%20Object%20Reference))
  (Rendering|Material|SetMaterial (Variables|Default|GetMesh) 1 NewCurrentMaterial)
  (CallFunction|ReleaseDynamicMaterial)
  (Variables|Materials|SetCurrentMaterial NewCurrentMaterial))

(fn SetInstigator
  (params (InInstigator :type Actor%20Object%20Reference))
  (Utilities|FlowControl|Sequence
    (:then_0
      (bind _ascontroller (Utilities|Casting|CastToController (Variables|GetInInstigator))
        (:then
          (LocalVariables|Setlocal%20Instigator _ascontroller))
        (:CastFailed
          (if (Utilities|DoesObjectImplementInterface (Variables|GetInInstigator) "/Script/FortniteGame.FortVehicleInterface")
            (bind _returnvalue (Vehicle|GetDriver(Message) (Variables|GetInInstigator)))
            (Utilities|IsValid _returnvalue
              (:Is%20Valid
                (LocalVariables|Setlocal%20Instigator (Pawn|GetController _returnvalue)))
              (:Is%20Not%20Valid
                (bind _returnvalue_2 (Vehicle|GetAllPassengers(Message) (Variables|GetInInstigator)))
                (bind _array_element (Utilities|Array|ForEachLoopwithBreak _returnvalue_2)
                  (:LoopBody
                    (Utilities|IsValid _array_element
                      (:Is%20Valid
                        (Utilities|IsValid (Pawn|GetController _array_element)
                          (:Is%20Valid
                            (LocalVariables|Setlocal%20Instigator (Pawn|GetController _array_element))))))))))
            (else
              (bind _aspawn (Utilities|Casting|CastToPawn (Variables|GetInInstigator)))
              (LocalVariables|Setlocal%20Instigator (Pawn|GetController _aspawn)))))))
    (:then_1
      ;[1 1 1] Sets separately rather than inline so if all of the instigator checks fail, we overwrite previous data to Null.
      (Variables|Default|SetTriggerInstigator (LocalVariables|Getlocal%20Instigator)))))

(fn SetMaterialBasedOnCurrentState
  (switch Utilities|FlowControl|Switch|SwitchonEEnabledState (Variables|Default|GetCurrentState)
    (:NewEnumerator0
      (CallFunction|SwitchToonMaterial))
    (:NewEnumerator1
      (CallFunction|SwitchToAnimatingMaterial))
    (:NewEnumerator2
      (CallFunction|SwitchToOffMaterial))))

(fn SetMeshVisibilityAndCollision
  (params (bNewVisibility :type Boolean))
  (bind _mesh (Variables|Default|GetMesh))
  (bind _mesh_2 (Variables|Default|GetMesh))
  (Development|SetHiddeninGame (Variables|Default|GetChannelNumberPlane) (not (and (Variables|GetbNewVisibility) (Variables|ChannelNumberUpdate|GetUsesChannelNumber))) false)
  (if (Variables|GetbNewVisibility)
    (Collision|SetCollisionProfileName _mesh "FortDynamicMesh" true :!class /Script/Engine.PrimitiveComponent)
    (Collision|SetCollisionResponsetoChannel _mesh "ECC_GameTraceChannel11" "ECR_Ignore")
    (Collision|SetCollisionResponsetoChannel _mesh "ECC_GameTraceChannel17" "ECR_Block")
    (elif (Variables|Options|GetbReceiveDamageWhenInvisible)
      (Collision|SetCollisionProfileName _mesh_2 "FortTriggerOnlyWeapons" true :!class /Script/Engine.PrimitiveComponent)
      (Collision|SetCollisionResponsetoChannel _mesh_2 "ECC_GameTraceChannel8" "ECR_Block")
      (else
        (Collision|SetCollisionEnabled _mesh_2 "NoCollision")))))

(fn StartActivateAttempt
  (params (TriggerInstigator :type Controller%20Object%20Reference) (DoTrigger :type Delegate%20%28by%20ref%29))
  ;[0 0.432 1] Attempting to Activate the Device, checking against if it can trigger or not.
  (Networking|SwitchHasAuthority
    (:Authority
      (if (DEPRECATED|CanActorInteract (Variables|Default|GetCreative_UseRestriction_Component) TriggerInstigator (CallFunction|CanTrigger) false)
        (Default|UnbindallEventsfromOnTrigger)
        (Default|BindEventtoOnTrigger DoTrigger)
        (CallFunction|SetInstigator TriggerInstigator)
        (Variables|Default|SetbIsResetting true)
        (if (> (Variables|Options|GetTriggerDelay) 0.0)
          (bind _returnvalue (Utilities|Time|SetTimerbyEvent (EventDispatchers|CreateEvent :!event SendActivationEvent) 0.01 false false (Variables|Options|GetTriggerDelay)))
          (Variables|Default|SetTriggerDelayTimer _returnvalue)
          (Variables|Default|SetCurrentState "NewEnumerator1")
          (else
            (CallFunction|SendActivationEvent)))))))

(fn TransmitMessage
  ;[1 1 1] Activate the Trigger that is attempting to fire.
  (Default|CallOnTrigger (Variables|Default|GetTriggerInstigator))
  (Default|UnbindallEventsfromOnTrigger)
  (return true))

(fn UpdateDynamicMaterial
  (Utilities|FlowControl|Sequence
    (:then_0
      (bind _channel_material_instance (Variables|Materials|GetChannel%20Material%20Instance)
        (:then
          ;[1 1 1] Material is getting disconnected from plane on object copy. This re-attaches it.
          (if (not (Utilities|Equal(Object) (Rendering|Material|GetMaterial (Variables|Default|GetChannelNumberPlane) 0) _channel_material_instance))
            (Rendering|Material|SetMaterial (Variables|Default|GetChannelNumberPlane) 0 (Variables|Materials|GetChannel%20Material%20Instance))))
        (:else
          (Utilities|BuildMaterial (Variables|Default|GetChannelNumberPlane) 0 0 (Variables|Materials|GetChannel%20Material%20Instance)))))
    (:then_1
      (bind _returnvalue (Class|FortMaterialCacheManager|GetCache (Variables|ChannelNumberUpdate|GetMaterial%20Cache%20Settings) (Variables|Materials|GetDefaultOnMaterial)))
      (Variables|Materials|SetMaterial%20Cache _returnvalue))))

(fn UpdateChannelNumber
  (params (Num :type Integer))
  (if (Variables|ChannelNumberUpdate|GetUsesChannelNumber)
    (ChannelNumberUpdate|UpdateDynamicMaterial)
    (Rendering|Material|SetScalarParameterValue (Variables|Materials|GetChannel%20Material%20Instance) "Score" (Math|Conversions|ToFloat(Integer) Num) :!class /Script/Engine.MaterialInstanceDynamic)
    (Rendering|Material|SetVectorParameterValue (Variables|Materials|GetChannel%20Material%20Instance) "Color" (Math|Color|SelectColor (Variables|ChannelNumberUpdate|GetChannelLEDColor) (Variables|ChannelNumberUpdate|GetChannelLEDColor_Unset) (> Num 0)) :!class /Script/Engine.MaterialInstanceDynamic)))

(fn ConstructionScript
  (Variables|BuildingActor|SetbIsInvulnerable true)
  (|CallParentFunction))
