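; Creative perception trigger device that fires separate events when it sees or loses sight of players, or players look at or away from it.
; The server takes AI sight updates, tests view angle on a timer, and applies per-event trigger counts and limits.
; Sight updates for players outside the minigame are queued until spawn.
; A throttled timer rotates the eye toward the tracked player, and perception state drives materials, beam, cues and audio.
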(event UserOptions|EventMigrateLegacyUserOptions
  (bind _transmiteveryxtriggers (Class|GetClassDefaults "/Game/Creative/Devices/Trigger/BP_Creative_Trigger_Parent.BP_Creative_Trigger_Parent_C"))
  (bind _transmiteveryxtriggers_2 (Variables|Options|GetTransmitEveryXTriggers))
  (|CallParentFunction)
  (if (!= (Variables|Options|GetTransmitEveryXTriggers) _transmiteveryxtriggers)
    (Utilities|FlowControl|Sequence
      (:then_0
        (Variables|UserOptions|Advanced|SetPlayer%20Looked%20At%20Transmit%20Every%20X%20Triggers _transmiteveryxtriggers_2))
      (:then_1
        (Variables|UserOptions|Advanced|SetDevice%20Loses%20Sight%20Of%20A%20PlayerTransmit%20Every%20X%20Triggers%20 _transmiteveryxtriggers_2))
      (:then_2
        (Variables|UserOptions|Advanced|SetPlayer%20Looked%20Away%20Transmit%20Every%20X%20Triggers _transmiteveryxtriggers_2))
      (:then_3
        (Variables|UserOptions|Advanced|SetDevice%20Sees%20A%20Player%20Transmit%20Every%20X%20Triggers _transmiteveryxtriggers_2))
      (:then_4
        (Variables|Options|SetTransmitEveryXTriggers _transmiteveryxtriggers))
      (:then_5
        (OptionData|RefreshPropertyOverrides (Variables|Default|GetToyOptionsComponent))))))

(event Custom|Set%20Tick
  (params (Enabled :type Boolean))
  ;[1 1 1] Server does not need to tick every frame as it just needs to be pointing roughly in the right direction
  ; (so if multiple clients move into range it will track the closest). As such, tick at a much lower rate.
  (if Enabled
    (bind _returnvalue (Class|CreativeBlueprintLibrary|SetLoopingTimerMaxOncePerFrame (delegate Custom|Tick) (select (Networking|HasAuthority) (Variables|Tick|GetServer%20Tick%20Time) (Variables|Tick|GetClient%20Tick%20Time))))
    (Variables|Tick|SetTick%20Handle _returnvalue)
    (else
      (Utilities|Time|ClearandInvalidateTimerbyHandle (Variables|Tick|GetTick%20Handle)))))

(event Custom|Tick
  ;[1 1 1] Ticking rotation to target. Tick is enabled or disabled based on if there are any perceived targets.
  (CallFunction|RotateToTarget))

(event OnTargetPerceptionUpdated(FortAIPerception)
  (params (Actor :type Actor%20Object%20Reference) (Stimulus :type AIStimulus%20Structure))
  ;[0 1 0.7599 0.5] Perception Trigger has Spotted a Player
  (Networking|SwitchHasAuthority
    (:Authority
      (bind _returnvalue (AI|Perception|GetSenseClassforStimulus Stimulus))
      (Utilities|Casting|CastToFortPawn Actor)
      (bind (_valid_controller _fortplayerpawn _playercontroller) (DEPRECATED|ActorHasValidController (Variables|Default|GetCreative_UseRestriction_Component) Actor true true))
      (if _valid_controller
        (CallFunction|OnPawnPerceptionUpdated _playercontroller _returnvalue (Utilities|Struct|BreakAIStimulus Stimulus))
        (CallFunction|UpdateTargetPlayer)))))

(event EventEnable
  ;[1 1 1] Enable or Disable the Perception component when the device is switched on / off
  (Networking|SwitchHasAuthority
    (:Authority
      (if (CallFunction|IsMinigameinProgress)
        (if (Utilities|Enum|Equal(Enum) (Variables|Default|GetPlayers%20Visible) "NewEnumerator3")
          (Variables|Default|SetPlayers%20Visible "NewEnumerator0"))))
    (:Remote
      _))
  (if (not (Variables|Default|GetbIsEnabled))
    (|CallParentFunction)
    (CallFunction|ResetDevice)))

(event PawnNowHasController(Creative_UseRestriction_Component)
  (params (FortPlayerPawn :type Fort%20Player%20Pawn%20Object%20Reference) (Player%20Controller :type Fort%20Player%20Controller%20Object%20Reference))
  ;[0 0.5795 1] Update Player Controller Perception when a valid event happens that may affect our perception of it.
  ; 1) It gains a controller. 2) It Switches Team. 3) It Switches Class 4) It Dies
  (CallFunction|UpdatePCPerception Player%20Controller))

(event Custom|OnPlayerDied
  (params (DamagedActor :type Actor%20Object%20Reference) (Damage :type Float%20%28single-precision%29) (InstigatedBy :type Controller%20Object%20Reference) (DamageCauser :type Actor%20Object%20Reference) (HitLocation :type Vector) (FHitComponent :type Primitive%20Component%20Object%20Reference) (BoneName :type Name) (Momentum :type Vector))
  ;[0 0.5795 1] Update Player Controller Perception when a valid event happens that may affect our perception of it.
  ; 1) It gains a controller. 2) It Switches Team. 3) It Switches Class 4) It Dies
  (bind _asplayer_controller (Utilities|Casting|CastToPlayerController (PlayertoPawnLink|GetLastPlayerController (Variables|Default|GetCreative_UseRestriction_Component) DamagedActor)))
  (CallFunction|UpdatePCPerception _asplayer_controller))

(event Custom|Player%20Class%20Switched
  (params (PlayerState :type Fort%20Player%20State%20Object%20Reference) (ClassSlotIndex :type Byte))
  ;[0 0.5795 1] Update Player Controller Perception when a valid event happens that may affect our perception of it.
  ; 1) It gains a controller. 2) It Switches Team. 3) It Switches Class 4) It Dies
  (CallFunction|UpdatePCPerception (Fortnite|GetFortPlayerControllerfromActor PlayerState)))

(event EventDisable
  ;[1 1 1] Enable or Disable the Perception component when the device is switched on / off
  (Networking|SwitchHasAuthority
    (:Authority
      (if (CallFunction|IsMinigameinProgress)
        (Variables|Default|SetPlayers%20Visible "NewEnumerator3")))
    (:Remote
      _))
  (if (Variables|Default|GetbIsEnabled)
    (|CallParentFunction)
    (CallFunction|ResetDevice)))

(event Custom|Player%20Switched%20Team
  (params (PlayerState :type Fort%20Player%20State%20Object%20Reference) (TeamId :type Byte) (OldTeamId :type Byte) (bTeamChangedWithoutRespawn :type Boolean))
  ;[0 0.5795 1] Update Player Controller Perception when a valid event happens that may affect our perception of it.
  ; 1) It gains a controller. 2) It Switches Team. 3) It Switches Class 4) It Dies
  (CallFunction|UpdatePCPerception (Fortnite|GetFortPlayerControllerfromActor PlayerState)))

(event Custom|BroadcastDeviceSawPlayer
  (params (Instigator :type Controller%20Object%20Reference))
  (bind _returnvalue (GameplayEvent|BroadcastEvent (Variables|UserOptions|Events|GetOn%20Device%20Sees%20a%20Player) 0 (Utilities|Struct|MakeGameplayEventDeviceSawPlayer Instigator)))
  (if (Math|Boolean|ANDBoolean _returnvalue (Variables|Default|GetbChildHandlesDecrement) (Variables|UserOptions|Advanced|GetDeviceSeesAPlayerTimesCanTrigger_Override))
    (Variables|Default|SetDeviceSeesAPlayerTimesCanTrigger_Remaining (- (Variables|Default|GetDeviceSeesAPlayerTimesCanTrigger_Remaining) 1))
    (Variables|Default|SetRemainingTimesCanTrigger (- (Variables|Default|GetRemainingTimesCanTrigger) 1))))

(event OnAnyPropertyChanged(ToyOptionsComponent)
  ;[0.04342 1 0.1768] Enable or Disable Perception component when property is changed.
  (CallFunction|ResetDevice)
  (CallFunction|UpdateVisualizationState)
  (Networking|SwitchHasAuthority
    (:Authority
      (Variables|Default|SetPlayers%20Visible (select (Variables|Options|GetEnabledOnMinigameStart) "NewEnumerator0" "NewEnumerator3"))
      (Utilities|Array|Clear (Variables|Default|GetDelayedPawnPerceptionUpdates)))))

(event MinigameEndedDelegate(FortMinigameLogic)
  ;[0.04342 1 0.1768] Enable or Disable Perception component when property is changed.
  (Variables|Default|SetPlayers%20Visible (select (Variables|Options|GetEnabledOnMinigameStart) "NewEnumerator0" "NewEnumerator3"))
  (Utilities|Array|Clear (Variables|Default|GetDelayedPawnPerceptionUpdates)))

(event Custom|BroadcastPlayerLookedAtDvice
  (params (Instigator :type Controller%20Object%20Reference))
  (bind _returnvalue (GameplayEvent|BroadcastEvent (Variables|UserOptions|Events|GetOn%20a%20Player%20Looks%20at%20Device) 0 (Utilities|Struct|MakeGameplayEventPlayerLookedatDvice Instigator)))
  (if (Math|Boolean|ANDBoolean _returnvalue (Variables|Default|GetbChildHandlesDecrement) (Variables|UserOptions|Advanced|GetPlayerLookedAtTimesCanTrigger_Override))
    (Variables|Default|SetPlayerLookedAtTimesCanTrigger_Remaining (- (Variables|Default|GetPlayerLookedAtTimesCanTrigger_Remaining) 1))
    (Variables|Default|SetRemainingTimesCanTrigger (- (Variables|Default|GetRemainingTimesCanTrigger) 1))))

(event OnPropertiesLoaded(ToyOptionsComponent)
  ;[0.04342 1 0.1768] Enable or Disable Perception component when property is changed.
  (CallFunction|ResetDevice)
  (CallFunction|UpdateVisualizationState)
  (Networking|SwitchHasAuthority
    (:Authority
      (Variables|Default|SetPlayers%20Visible (select (Variables|Options|GetEnabledOnMinigameStart) "NewEnumerator0" "NewEnumerator3"))
      (Utilities|Array|Clear (Variables|Default|GetDelayedPawnPerceptionUpdates)))))

(event Custom|BroadcastDeviceLostSightOfPlayer
  (params (Instigator :type Controller%20Object%20Reference))
  (bind _returnvalue (GameplayEvent|BroadcastEvent (Variables|UserOptions|Events|GetOn%20Device%20Loses%20Sight%20of%20a%20Player) 0 (Utilities|Struct|MakeGameplayEventDeviceLostSightOfPlayer Instigator)))
  (if (Math|Boolean|ANDBoolean _returnvalue (Variables|Default|GetbChildHandlesDecrement) (Variables|UserOptions|Advanced|GetDeviceLosesSightOfAPlayerTimesCanTrigger_Override))
    (Variables|Default|SetDeviceLosesSightOfAPlayerTimesCanTrigger_Remaining (- (Variables|Default|GetDeviceLosesSightOfAPlayerTimesCanTrigger_Remaining) 1))
    (Variables|Default|SetRemainingTimesCanTrigger (- (Variables|Default|GetRemainingTimesCanTrigger) 1))))

(event Custom|BroadcastPlayerLookedAwayFromDevice
  (params (Instigator :type Controller%20Object%20Reference))
  (bind _returnvalue (GameplayEvent|BroadcastEvent (Variables|UserOptions|Events|GetOn%20a%20Player%20Looks%20Away%20from%20Device) 0 (Utilities|Struct|MakeGameplayEventPlayerLookedAwayfromDevice Instigator)))
  (if (Math|Boolean|ANDBoolean _returnvalue (Variables|Default|GetbChildHandlesDecrement) (Variables|UserOptions|Advanced|GetPlayerLookedAwayTimesCanTrigger_Override))
    (Variables|Default|SetPlayerLookedAwayTimesCanTrigger_Remaining (- (Variables|Default|GetPlayerLookedAwayTimesCanTrigger_Remaining) 1))
    (Variables|Default|SetRemainingTimesCanTrigger (- (Variables|Default|GetRemainingTimesCanTrigger) 1))))

(event EventSetCurrentConstantMaterial
  (params (NewCurrentMaterial :type Material%20Interface%20Object%20Reference)))

(event Custom|MinigameStateChanged
  (params (Minigame :type Fort%20Minigame%20Object%20Reference) (MinigameState :type EFortMinigameState%20Enum :default "PreGame"))
  ;[0 0.6349 1] Updating Visualization when state changes.
  (CallFunction|UpdateVisualizationState))

(event EventSetMeshVisibilityAndCollision
  (params (bNewVisibility :type Boolean))
  ;[1 1 1] Set the collision of the one colliding object
  (Collision|SetCollisionEnabled (Variables|Default|GetCollision) (select bNewVisibility "QueryAndPhysics" "NoCollision")))

(event MinigameStartedDelegate(FortMinigameLogic)
  (CallFunction|ResetDevice)
  (Utilities|Array|Clear (Variables|Default|GetDelayedPawnPerceptionUpdates)))

(event EventBeginPlay
  (Visibility|SetVisibilityComponents (Variables|Default|GetVisibleInGameComponent) (Utilities|Array|MakeArray (Variables|Default|GetMesh) (Variables|Default|GetChannelNumberPlane) (Variables|Default|GetP_Trigger_Activate) (Variables|Default|GetRange%20Visualization)) true))

(event OnAddedtoMinigame(FortMinigameLogic)
  (params (Minigame :type Fort%20Minigame%20Object%20Reference))
  (CallFunction|BindorUnbindDelegates true))

(event EventEndPlay
  (params (EndPlayReason :type EEndPlayReason%20Enum :default "Destroyed"))
  (CallFunction|BindorUnbindDelegates false))

(event Custom|UpdateScreenVisuals
  (bind _returnvalue (Utilities|Time|GetGameTimeinSeconds))
  (bind _eyebeam (Variables|Default|GetEyeBeam))
  (Networking|CanExecuteCosmeticEvents
    (:True
      (switch Utilities|FlowControl|Switch|SwitchonEPerceptionState (Variables|Default|GetPlayers%20Visible)
        (:NewEnumerator0
          _)
        (:NewEnumerator1
          _)
        (:NewEnumerator2
          (Rendering|Material|SetScalarParameterValueonMaterials (Variables|Default|GetMesh) "IsActive" 1.0)
          (Rendering|Material|SetScalarParameterValueonMaterials (Variables|Default|GetMesh) "Blinking" 0.0)
          (Rendering|Material|SetVectorParameterValueonMaterials (Variables|Default|GetMesh) "BaseColorA" (Math|Conversions|ToVector(LinearColor) (Variables|Colors|GetAwareColorA)))
          (Rendering|Material|SetVectorParameterValueonMaterials (Variables|Default|GetMesh) "BaseColorB" (Math|Conversions|ToVector(LinearColor) (Variables|Colors|GetAwareColorB)))
          (Rendering|Material|SetVectorParameterValueonMaterials (Variables|Default|GetMesh) "EyeColorLow" (Math|Conversions|ToVector(LinearColor) (Variables|Colors|GetAwareEyeColorLow)))
          (Rendering|Material|SetVectorParameterValueonMaterials (Variables|Default|GetMesh) "EyeColorHigh" (Math|Conversions|ToVector(LinearColor) (Variables|Colors|GetAwareEyeColorHigh)))
          (Rendering|Material|SetScalarParameterValueonMaterials (Variables|Default|GetMesh) "Sighted" 1.0)
          (Rendering|Material|SetScalarParameterValueonMaterials (Variables|Default|GetMesh) "TimeOffset" (Math|Float|Fraction (Utilities|Time|GetGameTimeinSeconds)))
          (Development|SetHiddeninGame (Variables|Default|GetEyeBeam) (not (Variables|Options|GetVisibilityInGame)) false)
          (CallFunction|ApplyGameplayCuetoLocalClient)
          (Rendering|Material|SetCustomPrimitiveDataFloat _eyebeam 0 _returnvalue)
          (Rendering|Material|SetCustomPrimitiveDataFloat _eyebeam 1 (+ _returnvalue 999999.0))
          (Rendering|Material|SetVectorParameterValueonMaterials (Variables|Default|GetMesh) "Emissive" (Math|Conversions|ToVector(LinearColor) (Variables|Colors|GetDeviceAwareColor)))
          (bind _returnvalue_2 (Utilities|Time|SetTimerbyEvent (delegate Custom|ResetSightedParameter) 1.0))
          (Variables|Default|SetResetSightedParameterHandle _returnvalue_2))
        (:NewEnumerator3
          (Rendering|Material|SetScalarParameterValueonMaterials (Variables|Default|GetMesh) "IsActive" 0.0)
          (Rendering|Material|SetScalarParameterValueonMaterials (Variables|Default|GetMesh) "Blinking" 1.0 :!label _label)
          (Rendering|Material|SetVectorParameterValueonMaterials (Variables|Default|GetMesh) "BaseColorA" (Math|Conversions|ToVector(LinearColor) (Variables|Colors|GetUnawareColorA)))
          (Rendering|Material|SetVectorParameterValueonMaterials (Variables|Default|GetMesh) "BaseColorB" (Math|Conversions|ToVector(LinearColor) (Variables|Colors|GetUnawareColorB)))
          (Rendering|Material|SetVectorParameterValueonMaterials (Variables|Default|GetMesh) "EyeColorLow" (Math|Conversions|ToVector(LinearColor) (Variables|Colors|GetUnawareEyeColorLow)))
          (Rendering|Material|SetVectorParameterValueonMaterials (Variables|Default|GetMesh) "EyeColorHigh" (Math|Conversions|ToVector(LinearColor) (Variables|Colors|GetUnawareEyeColorHigh)))
          (Rendering|Material|SetScalarParameterValueonMaterials (Variables|Default|GetMesh) "Sighted" 0.0)
          (Utilities|Time|ClearandInvalidateTimerbyHandle (Variables|Default|GetResetSightedParameterHandle))
          (CallFunction|RemoveGameplayCueonLocalClient)
          (Rendering|Material|SetCustomPrimitiveDataFloat (Variables|Default|GetEyeBeam) 1 (Utilities|Time|GetGameTimeinSeconds))
          (Development|SetHiddeninGame (Variables|Default|GetEyeBeam) true false)
          (Rendering|Material|SetVectorParameterValueonMaterials (Variables|Default|GetMesh) "Emissive" (Math|Conversions|ToVector(LinearColor) (Variables|Colors|GetDeviceUnawareColor)))))
      (Rendering|Material|SetScalarParameterValueonMaterials (Variables|Default|GetMesh) "IsActive" 1.0)
      (goto _label))))

(event Custom|ResetSightedParameter
  (Rendering|Material|SetScalarParameterValueonMaterials (Variables|Default|GetMesh) "Sighted" 0.0))

(event OnComponentBeginOverlap(RangeVisualizationTrigger)
  (params (OverlappedComponent :type Primitive%20Component%20Object%20Reference) (OtherActor :type Actor%20Object%20Reference) (OtherComp :type Primitive%20Component%20Object%20Reference) (OtherBodyIndex :type Integer) (bFromSweep :type Boolean) (SweepResult :type Hit%20Result%20Structure%20%28by%20ref%29))
  ;[1 0.8995 0 0.25] Show Visualization of Range when Near
  (if (not (Building|IsPreviewBuildingActor))
    (if (HelperFunctions|IsLocalPawn OtherActor)
      (CallFunction|ShowRangeVisualization))))

(event OnComponentEndOverlap(RangeVisualizationTrigger)
  (params (OverlappedComponent :type Primitive%20Component%20Object%20Reference) (OtherActor :type Actor%20Object%20Reference) (OtherComp :type Primitive%20Component%20Object%20Reference) (OtherBodyIndex :type Integer))
  ;[1 0.8995 0 0.25] Show Visualization of Range when Near
  (if (not (Building|IsPreviewBuildingActor))
    (if (HelperFunctions|IsLocalPawn OtherActor)
      (CallFunction|HideRangeVisualization))))

(event Custom|ShowRangeVisualization
  ;[1 0.8995 0 0.25] Show Visualization of Range when Near
  (Networking|CanExecuteCosmeticEvents
    (:True
      (Development|SetHiddeninGame (Variables|Default|GetRange%20Visualization) false false)
      (Audio|PlaySoundatLocation (Utilities|ResolveSoftReference (Utilities|Array|Get(acopy) (Variables|SFX|GetTriggerSounds))) (Transformation|GetActorLocation))
      (bind (_direction _lerp) (Timeline|Visualization%20Range%20%25 Play)
        (:Update
          (Rendering|Material|SetScalarParameterValue (Material|GetMID (Variables|Default|GetRange%20Visualization)) "Opacity" _lerp :!class /Script/Engine.MaterialInstanceDynamic)
          (Transformation|SetWorldScale3D (Variables|Default|GetRange%20Visualization) (Math|Vector|Lerp(Vector) "0.010000,0.010000,0.010000" (Math|Conversions|ToVector(Float) (* (Variables|Default|GetVisual%20Range) 2.0)) _lerp)))
        (:Finished
          (switch Utilities|FlowControl|Switch|SwitchonETimelineDirection _direction
            (:Backward
              (Development|SetHiddeninGame (Variables|Default|GetRange%20Visualization) true false))))))))

(event Custom|HideRangeVisualization
  ;[1 0.8995 0 0.25] Show Visualization of Range when Near
  (Networking|CanExecuteCosmeticEvents
    (:True
      (Audio|PlaySoundatLocation (Utilities|ResolveSoftReference (Utilities|Array|Get(acopy) (Variables|SFX|GetTriggerSounds) 1)) (Transformation|GetActorLocation))
      (Timeline|Visualization%20Range%20%25 Reverse))))

(fn ApplyGameplayCueToLocalClient
  (Variables|Default|GetLastPawnGameplayCueAppliedTo
    (:then
      (CallFunction|RemoveGameplayCueonLocalClient) ; If another player currently has a cue from this device, remove it.
    )
    (:else
      _))
  (bind _tracked_player (Variables|Default|GetTracked%20Player)
    (:then
      (bind _asfort_player_pawn (Utilities|Casting|CastToFortPlayerPawn _tracked_player))
      (bind _output_get (Variables|Default|SetLastPawnGameplayCueAppliedTo _asfort_player_pawn))
      (CallFunction|PlaySpottedAudio (Pawn|IsLocallyControlled _output_get) true)
      (if (Variables|Default|GetTriggerVFX)
        (Projectile|AddGameplayCue%28Looping%29Local (Variables|Default|GetLastPawnGameplayCueAppliedTo) (Variables|Default|GetGameplayCueTagActive) (Ability|GameplayCue|MakeGameplayCueParameters 0.0))))
    (:else)))

(fn AttemptActivateIfSet
  (params (Trigger :type Fort%20Gameplay%20Trigger%20Message%20Component%20Object%20Reference) (TriggerInstigator :type Controller%20Object%20Reference) (DoTrigger :type Delegate%20%28by%20ref%29))
  (CallFunction|StartActivateAttempt TriggerInstigator DoTrigger))

(fn BindOrUnbindDelegates
  (params (Bind :type Boolean))
  (bind _outputdelegate (EventDispatchers|CreateEvent :!event OnMinigamePlayerSpawned))
  (bind _outputdelegate_2 (EventDispatchers|CreateEvent :!event MinigameStateChanged))
  (bind _outputdelegate_3 (EventDispatchers|CreateEvent :!event Player%20Switched%20Team))
  (bind _outputdelegate_4 (EventDispatchers|CreateEvent :!event Player%20Class%20Switched))
  (bind _fortminigamelogic (Variables|Default|GetFortMinigameLogic)
    (:then
      (bind _currentminigame (Variables|/Script/FortniteGame.FortMinigameLogicComponent|GetCurrentMinigame _fortminigamelogic)
        (:then
          (if Bind
            (Minigame|BindEventtoOnPlayerClassApplied _currentminigame _outputdelegate_4)
            (Minigame|BindEventtoOnPlayerClassRemoved _currentminigame _outputdelegate_4)
            (Minigame|BindEventtoStateChangedDelegate _currentminigame _outputdelegate_2 :!class /Script/FortniteGame.FortMinigame)
            (Minigame|BindEventtoPlayerChangedTeamDelegate _currentminigame _outputdelegate_3)
            (Minigame|BindEventtoOnMinigamePlayerSpawned _currentminigame _outputdelegate)
            (else
              (Minigame|UnbindEventfromOnPlayerClassApplied _currentminigame _outputdelegate_4)
              (Minigame|UnbindEventfromOnPlayerClassRemoved _currentminigame _outputdelegate_4)
              (Minigame|UnbindEventfromStateChangedDelegate _currentminigame _outputdelegate_2 :!class /Script/FortniteGame.FortMinigame)
              (Minigame|UnbindEventfromPlayerChangedTeamDelegate _currentminigame _outputdelegate_3)
              (Minigame|UnbindEventfromOnMinigamePlayerSpawned _currentminigame _outputdelegate))))
        (:else)))
    (:else)))

(fn CheckPlayersToDeviceLOS
  (if (CallFunction|CanTrigger)
    (Utilities|FlowControl|Sequence
      (:then_0
        ;[1 1 1] Check all players that the device can see to see if they can see it.
        ; If they can move them to the Looking At device Array.
        (bind (_arrayindex _arrayelement) (Utilities|Array|ReverseforEachLoop (Variables|Default|GetCurrentlySightedPlayers))
          (:LoopBody
            (if (CallFunction|IsPlayerLookingAtDevice _arrayelement)
              (bind _should_activate_event (CallFunction|UpdateTriggerCountandCheckRemainingUses (Variables|Default|GetPlayer%20Looked%20At%20Transmit%20Every%20X%20Triggers%20Count) (Variables|UserOptions|Advanced|GetPlayer%20Looked%20at%20Transmit%20Every%20X%20Triggers) (Variables|Default|GetPlayerLookedAtTimesCanTrigger_Remaining) (Variables|UserOptions|Advanced|GetPlayerLookedAtTimesCanTrigger_Override)))
              (if _should_activate_event
                (CallFunction|StartActivateAttempt _arrayelement (EventDispatchers|CreateEvent :!event BroadcastPlayerLookedAtDvice)))
              (Utilities|Array|Add (Variables|Default|GetPlayersLookingAtDevice) _arrayelement)
              (Utilities|Array|RemoveIndex (Variables|Default|GetCurrentlySightedPlayers) _arrayindex)))))
      (:then_1
        ;[1 1 1] Check all player that are looking at this device and check if they are no longer looking at it.
        ; If they are not looking at the device, then move them back to the sighted players array,.
        (bind (_arrayindex_2 _arrayelement_2) (Utilities|Array|ReverseforEachLoop (Variables|Default|GetPlayersLookingAtDevice))
          (:LoopBody
            (if (not (CallFunction|IsPlayerLookingAtDevice _arrayelement_2))
              (bind _should_activate_event_2 (CallFunction|UpdateTriggerCountandCheckRemainingUses (Variables|Default|GetPlayer%20Looked%20Away%20Transmit%20Every%20X%20Triggers%20Count) (Variables|UserOptions|Advanced|GetPlayer%20Looked%20Away%20Transmit%20Every%20X%20Triggers) (Variables|Default|GetPlayerLookedAwayTimesCanTrigger_Remaining) (Variables|UserOptions|Advanced|GetPlayerLookedAwayTimesCanTrigger_Override)))
              (if _should_activate_event_2
                (CallFunction|StartActivateAttempt _arrayelement_2 (EventDispatchers|CreateEvent :!event BroadcastPlayerLookedAwayFromDevice)))
              (Utilities|Array|Add (Variables|Default|GetCurrentlySightedPlayers) _arrayelement_2)
              (Utilities|Array|RemoveIndex (Variables|Default|GetPlayersLookingAtDevice) _arrayindex_2))))))))

(fn ClearPerceivedPlayers
  (Networking|SwitchHasAuthority
    (:Authority
      (Utilities|Array|Clear (Variables|Default|GetPlayersLookingAtDevice))
      (Utilities|Array|Clear (Variables|Default|GetCurrentlySightedPlayers))
      (Variables|Default|SetTracked%20Player 0)
      (AI|Perception|ForgetAll (Variables|Default|GetFortAIPerception)))))

(fn EyeBeamScale
  (params (TargetLocation :type Vector))
  (bind (_returnvalue_x _returnvalue_y _returnvalue_z) (Math|Vector|BreakVector (Transformation|GetWorldScale (Variables|Default|GetEyeBeam))))
  (return (Math|Vector|MakeVector (/ (Math|Vector|Distance(Vector) TargetLocation (Transformation|GetWorldLocation (Variables|Default|GetEyeBeam))) 1e+02) _returnvalue_y _returnvalue_z)))

(fn IsPlayerLookingatDevice
  (params (Player%20Controller :type Player%20Controller%20Object%20Reference))
  (Utilities|IsValid Player%20Controller
    (:Is%20Valid
      (bind _returnvalue (Fortnite|GetAngleDegrees (- (Transformation|GetActorLocation) (Camera|GetCameraLocation (Variables|/Script/Engine.PlayerController|GetPlayerCameraManager Player%20Controller))) (Math|Vector|GetForwardVector (Camera|GetCameraRotation (Variables|/Script/Engine.PlayerController|GetPlayerCameraManager Player%20Controller)))))
      (return (< _returnvalue (Variables|Default|GetPlayerFieldOfVisionHalfAngle))))
    (:Is%20Not%20Valid
      (return false))))

(fn IsAlreadyPerceived
  (params (PC :type Player%20Controller%20Object%20Reference%20%28by%20ref%29))
  (bind _output_get (LocalVariables|SetArrayToCheck (Variables|Default|GetCurrentlySightedPlayers)))
  (Utilities|Array|AppendArray _output_get (Variables|Default|GetPlayersLookingAtDevice))
  (return (Utilities|Array|ContainsItem (LocalVariables|GetArrayToCheck) (Variables|GetPC))))

(fn IsMinigameInProgress
  (bind _fortminigamelogic (Variables|Default|GetFortMinigameLogic)
    (:then
      (bind _currentminigame (Variables|/Script/FortniteGame.FortMinigameLogicComponent|GetCurrentMinigame _fortminigamelogic)
        (:then
          (return (Utilities|Enum|Equal(Enum) (Variables|/Script/FortniteGame.FortMinigame|GetCurrentState _currentminigame) "InProgress")))
        (:else
          _)))
    (:else
      _))
  (return false))

(fn LookAtRotation
  (params (TargetLocation :type Vector))
  (return (select (Math|Rotator|Equal(Rotator) (Transformation|GetWorldRotation (Variables|Default|GetMesh)) (Math|Interpolation|RInterpTo (Transformation|GetWorldRotation (Variables|Default|GetMesh)) (Math|Rotator|MakeRotfromX (- TargetLocation (Transformation|GetWorldLocation (Variables|Default|GetMesh)))) (Utilities|Time|GetWorldDeltaSeconds) (Variables|Default|GetTrack%20Rate)) (Variables|Default|GetRotation%20Tolerance)) (Transformation|GetWorldRotation (Variables|Default|GetMesh)) (Math|Interpolation|RInterpTo (Transformation|GetWorldRotation (Variables|Default|GetMesh)) (Math|Rotator|MakeRotfromX (- TargetLocation (Transformation|GetWorldLocation (Variables|Default|GetMesh)))) (Utilities|Time|GetWorldDeltaSeconds) (Variables|Default|GetTrack%20Rate)))))

(fn OnPawnPerceptionUpdated
  (params (Sensed%20Controller :type Controller%20Object%20Reference) (Sense%20Class :type Object%20Class%20Reference) (Successfully%20Sensed :type Boolean))
  (bind _successfully_sensed (Variables|GetSuccessfully%20Sensed))
  (bind _sense_class (Variables|GetSense%20Class))
  (bind _sensed_controller (Variables|GetSensed%20Controller))
  (bind _sdelayedpawnperceptionupdate (Utilities|Struct|MakeSDelayedPawnPerceptionUpdate _sensed_controller _sense_class _successfully_sensed))
  (bind _perceived_controller (LocalVariables|GetPerceived%20Controller))
  (bind _asplayer_controller (Utilities|Casting|CastToPlayerController Sensed%20Controller))
  (LocalVariables|SetPerceived%20Controller _asplayer_controller)
  (if (Utilities|Equal(Class) Sense%20Class "/Script/AIModule.AISense_Sight")
    (Utilities|ControllerinMinigame (LocalVariables|GetPerceived%20Controller)
      (:True
        ;[0.937 1 0.1334] Can the Device See the Player
        (if (Variables|GetSuccessfully%20Sensed)
          (if (not (CallFunction|IsAlreadyPerceived (LocalVariables|GetPerceived%20Controller)))
            (bind _should_activate_event (CallFunction|UpdateTriggerCountandCheckRemainingUses (Variables|Default|GetDevice%20Sees%20A%20Player%20Transmit%20Every%20X%20Triggers%20Count) (Variables|UserOptions|Advanced|GetDevice%20Sees%20A%20Player%20Transmit%20Every%20X%20Triggers) (Variables|Default|GetDeviceSeesAPlayerTimesCanTrigger_Remaining) (Variables|UserOptions|Advanced|GetDeviceSeesAPlayerTimesCanTrigger_Override)))
            (if _should_activate_event
              (CallFunction|StartActivateAttempt (LocalVariables|GetPerceived%20Controller) (EventDispatchers|CreateEvent :!event BroadcastDeviceSawPlayer)))
            ;[0 0.9036 1] Update is player looking at device
            (Utilities|Array|AddUnique (Variables|Default|GetCurrentlySightedPlayers) _perceived_controller)
            (Utilities|Casting|CastToFortPlayerPawnAthena (Pawn|GetControlledPawn _perceived_controller)
              (:then
                _)
              (:CastFailed
                _))
            (if (Utilities|Time|IsTimerActivebyHandle (Variables|Default|GetPlayerSightedUpdateTimer))
              (return)
              (else
                (bind _returnvalue (Utilities|Time|SetTimerbyEvent (EventDispatchers|CreateEvent :!event CheckPlayersToDeviceLOS) (Variables|Default|GetPlayerLOSUpdateTime) true))
                (Variables|Default|SetPlayerSightedUpdateTimer _returnvalue)
                (CallFunction|CheckPlayerstoDeviceLOS) ; Immediately trigger as well as starting the timer
                (Variables|Default|SetPlayers%20Visible "NewEnumerator2"))))
          (elif (CallFunction|IsAlreadyPerceived (LocalVariables|GetPerceived%20Controller))
            (bind _should_activate_event_2 (CallFunction|UpdateTriggerCountandCheckRemainingUses (Variables|Default|GetDevice%20Loses%20Sight%20Of%20A%20Player%20Transmit%20Every%20X%20Triggers%20Count) (Variables|UserOptions|Advanced|GetDevice%20Loses%20Sight%20Of%20A%20PlayerTransmit%20Every%20X%20Triggers%20) (Variables|Default|GetDeviceLosesSightOfAPlayerTimesCanTrigger_Remaining) (Variables|UserOptions|Advanced|GetDeviceLosesSightOfAPlayerTimesCanTrigger_Override)))
            (if _should_activate_event_2
              (CallFunction|StartActivateAttempt (LocalVariables|GetPerceived%20Controller) (EventDispatchers|CreateEvent :!event BroadcastDeviceLostSightOfPlayer)))
            (Utilities|Array|RemoveItem (Variables|Default|GetCurrentlySightedPlayers) (LocalVariables|GetPerceived%20Controller))
            (bind _returnvalue_2 (Utilities|Array|RemoveItem (Variables|Default|GetPlayersLookingAtDevice) (LocalVariables|GetPerceived%20Controller)))
            (if _returnvalue_2
              (bind _should_activate_event_3 (CallFunction|UpdateTriggerCountandCheckRemainingUses (Variables|Default|GetPlayer%20Looked%20Away%20Transmit%20Every%20X%20Triggers%20Count) (Variables|UserOptions|Advanced|GetPlayer%20Looked%20Away%20Transmit%20Every%20X%20Triggers) (Variables|Default|GetPlayerLookedAwayTimesCanTrigger_Remaining) (Variables|UserOptions|Advanced|GetPlayerLookedAwayTimesCanTrigger_Override)))
              (if _should_activate_event_3
                (CallFunction|StartActivateAttempt (LocalVariables|GetPerceived%20Controller) (EventDispatchers|CreateEvent :!event BroadcastPlayerLookedAwayFromDevice))))
            (if (and (== (Utilities|Array|Length (Variables|Default|GetCurrentlySightedPlayers)) 0) (== (Utilities|Array|Length (Variables|Default|GetPlayersLookingAtDevice)) 0))
              (Utilities|Time|ClearandInvalidateTimerbyHandle (Variables|Default|GetPlayerSightedUpdateTimer))
              (Variables|Default|SetPlayers%20Visible "NewEnumerator0")))))
      (:False
        ;[1 1 1] When the pawn that triggered the perception component isn't in the minigame, put it into the pending list so it can be handled when the pawn is back in the mini game
        (if (not (Utilities|Array|IsValidIndex (Variables|Default|GetDelayedPawnPerceptionUpdates) (Utilities|Array|FindItem (Variables|Default|GetDelayedPawnPerceptionUpdates) _sdelayedpawnperceptionupdate)))
          (Utilities|Array|Add (Variables|Default|GetDelayedPawnPerceptionUpdates) _sdelayedpawnperceptionupdate))))))

(fn OnMinigamePlayerSpawned
  (params (SpawningPlayer :type Controller%20Object%20Reference))
  ;[1 1 1] Bind to pawn death, so we can update perception then also.
  (bind _asfort_pawn (Utilities|Casting|CastToFortPawn (Pawn|GetControlledPawn SpawningPlayer))
    (:then
      (EventDispatchers|BindEventtoOnDied _asfort_pawn (EventDispatchers|CreateEvent :!event OnPlayerDied) :!class /Script/FortniteGame.FortPawn))
    (:CastFailed
      _))
  (if (CallFunction|IsMinigameinProgress)
    (bind (_arrayindex _arrayelement) (Utilities|Array|ReverseforEachLoop (Variables|Default|GetDelayedPawnPerceptionUpdates))
      (:LoopBody
        (bind (_arrayelement_sensedcontroller_11_b4abf07f4de3a57c7b3bac9c220075b8 _arrayelement_senseclass_12_cfcbee7d4fb9f54f54c3debdd2cddefc _arrayelement_successfullysensed_10_7882a390431f8721c8621483de61ed8c) (Utilities|Struct|BreakSDelayedPawnPerceptionUpdate _arrayelement))
        (if (Utilities|Equal(Object) _arrayelement_sensedcontroller_11_b4abf07f4de3a57c7b3bac9c220075b8 (Variables|GetSpawningPlayer))
          (CallFunction|OnPawnPerceptionUpdated _arrayelement_sensedcontroller_11_b4abf07f4de3a57c7b3bac9c220075b8 _arrayelement_senseclass_12_cfcbee7d4fb9f54f54c3debdd2cddefc _arrayelement_successfullysensed_10_7882a390431f8721c8621483de61ed8c)
          (Utilities|Array|RemoveIndex (Variables|Default|GetDelayedPawnPerceptionUpdates) _arrayindex))))))

(fn OnRep_PlayersVisible
  (CallFunction|PlayerSpottedEffects (Variables|Default|GetPlayers%20Visible)))

(fn OnRep_TrackedPlayer
  (bind _tracked_player (Variables|Default|GetTracked%20Player)
    (:then
      (Development|PrintString (Utilities|GetDisplayName _tracked_player))
      (CallFunction|SetTick true))
    (:else
      (Development|PrintString "Tracked Player Cleared"))))

(fn PlaySpottedAudio
  (params (Local%20Controller :type Boolean) (Player%20Spotted :type Boolean))
  (if (== (Variables|Options|GetTriggerSoundIndex) 0)
    (if Local%20Controller
      (Audio|PlaySound2D (Utilities|ResolveSoftReference (Utilities|Array|Get(acopy) (Variables|SFX|GetTriggerSounds) (select (Variables|GetPlayer%20Spotted) 2 3))))
      (else
        (Audio|PlaySoundatLocation (Utilities|ResolveSoftReference (Utilities|Array|Get(acopy) (Variables|SFX|GetTriggerSounds) (select (Variables|GetPlayer%20Spotted) 0 1))) (Transformation|GetActorLocation))))))

(fn PlayerSpottedEffects
  (params (Spotted :type EPerceptionState%20Enum :default "NewEnumerator0"))
  ;[1 0.7441 0] Switches eye material based on the state of the device
  (Networking|CanExecuteCosmeticEvents
    (:True
      (CallFunction|UpdateScreenVisuals))))

(fn RemoveGameplayCueOnLocalClient
  (bind _lastpawngameplaycueappliedto (Variables|Default|GetLastPawnGameplayCueAppliedTo)
    (:then
      (Projectile|RemoveGameplayCue%28Looping%29Local _lastpawngameplaycueappliedto (Variables|Default|GetGameplayCueTagActive) (Ability|GameplayCue|MakeGameplayCueParameters 0.0))
      (CallFunction|PlaySpottedAudio (Pawn|IsLocallyControlled _lastpawngameplaycueappliedto))
      (Variables|Default|SetLastPawnGameplayCueAppliedTo 0))
    (:else)))

(fn ResetDevice
  (bind _output_get (LocalVariables|SetEnable%20Perception (Default|ShouldPerceptionbeEnabled)))
  (if _output_get
    (CallFunction|ClearPerceivedPlayers)
    (CallFunction|SetPerceptionEnabled true)
    (else
      (CallFunction|SetPerceptionEnabled false)
      (CallFunction|ClearPerceivedPlayers))))

(fn ResetTriggerCount
  (|CallParentFunction)
  (Utilities|FlowControl|Sequence
    (:then_0
      (Variables|Default|SetPlayerLookedAtTimesCanTrigger_Remaining (Variables|UserOptions|Advanced|GetPlayerLookedAtTimesCanTrigger))
      (Variables|Default|SetPlayer%20Looked%20At%20Transmit%20Every%20X%20Triggers%20Count 0))
    (:then_1
      (Variables|Default|SetPlayerLookedAwayTimesCanTrigger_Remaining (Variables|UserOptions|Advanced|GetPlayerLookedAwayTimesCanTrigger))
      (Variables|Default|SetPlayer%20Looked%20Away%20Transmit%20Every%20X%20Triggers%20Count 0))
    (:then_2
      (Variables|Default|SetDeviceSeesAPlayerTimesCanTrigger_Remaining (Variables|UserOptions|Advanced|GetDeviceSeesAPlayerTimesCanTrigger))
      (Variables|Default|SetDevice%20Sees%20A%20Player%20Transmit%20Every%20X%20Triggers%20Count 0))
    (:then_3
      (Variables|Default|SetDeviceLosesSightOfAPlayerTimesCanTrigger_Remaining (Variables|UserOptions|Advanced|GetDeviceLosesSightOfAPlayerTimesCanTrigger))
      (Variables|Default|SetDevice%20Loses%20Sight%20Of%20A%20Player%20Transmit%20Every%20X%20Triggers%20Count 0))))

(fn RotatetoTarget
  (if (Variables|Default|GetbIsEnabled)
    (bind _tracked_player (Variables|Default|GetTracked%20Player)
      (:then
        (Utilities|IsValid _tracked_player
          (:Is%20Valid
            (bind _asfort_player_pawn (Utilities|Casting|CastToFortPlayerPawn _tracked_player)
              (:then
                (if (Transformation|DoesSocketExist (Variables|/Script/Engine.Character|GetMesh _asfort_player_pawn) "FX_Chest")
                  (LocalVariables|SetTarget%20Location (Transformation|GetSocketLocation (Variables|/Script/Engine.Character|GetMesh _asfort_player_pawn) "FX_Chest"))
                  (else
                    (LocalVariables|SetTarget%20Location (Transformation|GetActorLocation _tracked_player) :!label _label)))
                (Transformation|SetWorldRotation (Variables|Default|GetMesh) (CallFunction|LookatRotation (LocalVariables|GetTarget%20Location)) false false :!label _label_2)
                (Networking|CanExecuteCosmeticEvents
                  (:True
                    (Transformation|SetWorldScale3D (Variables|Default|GetEyeBeam) (CallFunction|EyeBeamScale (LocalVariables|GetTarget%20Location))))))
              (:CastFailed
                _))
            (goto _label))
          (:Is%20Not%20Valid
            (if (Math|Rotator|NotEqual(Rotator) (Variables|/Script/Engine.SceneComponent|GetRelativeRotation (Variables|Default|GetMesh)) "0, 0, 0" (Variables|Default|GetRotation%20Tolerance)) :!label _label_3
              (LocalVariables|SetTarget%20Location (Transformation|GetWorldLocation (Variables|Default|GetBaseRotationLocation)))
              (else
                (CallFunction|SetTick false)
                (Transformation|SetRelativeRotation (Variables|Default|GetMesh) "0, 0, 0" false false) ; We're very close to zero, but not quite (rinterp ticks for a long while before reaching exact).\nSnap the last bit.
              ))))
        (goto _label_2))
      (:else
        _)))
  (goto _label_3))

(fn SetPerceptionEnabled
  (params (Enable%20Perception%20System :type Boolean))
  ;[1 1 1] Enable / Disable Senses
  (AI|Perception|SetSenseEnabled (Variables|Default|GetFortAIPerception) "/Script/AIModule.AISense_Sight" (Variables|GetEnable%20Perception%20System))
  ;[1 1 1] Enable / Disable the whole Perception Component
  (Components|Activation|SetActive (Variables|Default|GetFortAIPerception) (Variables|GetEnable%20Perception%20System) true)
  (if (not (Variables|GetEnable%20Perception%20System))
    (Utilities|Time|ClearandInvalidateTimerbyHandle (Variables|Default|GetPlayerSightedUpdateTimer))))

(fn ShouldPerceptionBeEnabled
  (return (and (Class|CreativeCommonDeviceFunctionLibrary|IsinPlayMode self) (Variables|Default|GetbIsEnabled))))

(fn UpdatePCPerception
  (params (PC :type Player%20Controller%20Object%20Reference))
  ;[1 1 1] Tests whether the Player Controller is currently perceived and should be perceived.
  (Networking|SwitchHasAuthority
    (:Authority
      (Utilities|IsValid PC
        (:Is%20Valid
          (bind _asfort_pawn (Utilities|Casting|CastToFortPawn (Pawn|GetControlledPawn PC)))
          (LocalVariables|SetSensedPawn (and (Utilities|Array|ContainsItem (AI|Perception|GetCurrentlyPerceivedActors (Variables|Default|GetFortAIPerception) (LocalVariables|GetSense)) _asfort_pawn) (not (FortPawn|IsDead _asfort_pawn))))
          (Utilities|ControllerinMinigame (Variables|GetPC)
            (:True
              (CallFunction|OnPawnPerceptionUpdated (Variables|GetPC) (LocalVariables|GetSense) (DEPRECATED|CanPlayerControllerInteract (Variables|Default|GetCreative_UseRestriction_Component) (Variables|GetPC) (LocalVariables|GetSensedPawn) false))
              (CallFunction|UpdateTargetPlayer))))))))

(fn UpdateTargetPlayer
  (Networking|SwitchHasAuthority
    (:Authority
      (bind _output_get (LocalVariables|Setlist%20of%20players (Variables|Default|GetPlayersLookingAtDevice))) ; local variable
      (Utilities|Array|AppendArray _output_get (Variables|Default|GetCurrentlySightedPlayers))
      (bind _tracked_player (Variables|Default|GetTracked%20Player)
        (:then
          (Utilities|IsValid (Fortnite|GetFortPlayerControllerfromActor _tracked_player)
            (:Is%20Valid
              (if (Utilities|Array|ContainsItem (LocalVariables|Getlist%20of%20players) (Fortnite|GetFortPlayerControllerfromActor _tracked_player))
                _
                (else
                  _)))
            (:Is%20Not%20Valid
              _)))
        (:else
          _)) ; current Player Pawn\nbeing tracked.
      (for _array_element (LocalVariables|Getlist%20of%20players)
        (Utilities|IsValid (Pawn|GetControlledPawn _array_element)
          (:Is%20Valid
            (if (> (Math|Vector|DotProduct (Math|Vector|Normalize (- (Transformation|GetActorLocation (Pawn|GetControlledPawn _array_element)) (Transformation|GetWorldLocation (Variables|Default|GetMesh)))) (Transformation|GetForwardVector (Variables|Default|GetMesh))) (LocalVariables|GetBest%20Dot%20Result))
              (LocalVariables|SetBest%20Pawn (Pawn|GetControlledPawn _array_element))))))
      (Variables|Default|SetTracked%20Player (LocalVariables|GetBest%20Pawn)))))

(fn UpdateVisualizationState
  (Networking|CanExecuteCosmeticEvents
    (:True
      (if (not (Class|CreativeCommonDeviceFunctionLibrary|IsinPlayMode self))
        (Collision|SetCollisionEnabled (Variables|Default|GetRange%20Visualization%20Trigger) "QueryOnly")
        (else
          _)))
    (:False
      _))
  (Collision|SetCollisionEnabled (Variables|Default|GetRange%20Visualization%20Trigger) "NoCollision")
  (CallFunction|HideRangeVisualization))

(fn UpdateTriggerCountAndCheckRemainingUses
  (params (TimesTriggered :type Integer%20%28by%20ref%29) (Every%20N%20Triggers :type Integer) (Remaining%20Triggers :type Integer) (HasMaximumTriggers :type Boolean))
  (Variables|SetBy-RefVar TimesTriggered (Math|Integer|%25%28Integer%29 (+ TimesTriggered 1) Every%20N%20Triggers))
  (return (and (== (Variables|Default|GetTimesTriggered) 0) (or (> Remaining%20Triggers 0) (not HasMaximumTriggers)))))

(fn ConstructionScript
  (|CallParentFunction)
  (Networking|CanExecuteCosmeticEvents
    (:True
      (Transformation|AttachComponentToComponent (Variables|Default|GetEyeBeam) (Variables|Default|GetMesh) "LookFromSocket" "KeepRelative" "KeepRelative" "KeepRelative" true)
      (Rendering|Material|SetCustomPrimitiveDataFloat (Variables|Default|GetEyeBeam) 2 (Variables|Default|GetBeamTransitionLength)))))
