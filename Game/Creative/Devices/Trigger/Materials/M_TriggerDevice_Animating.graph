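# Animating material for the Creative Trigger device, pairing a dark grey base with a pulsing emissive glow.
# A sine wave over time since the StartingTime parameter, offset upward, drives a lerp between the Glow color (orange by default) and black.
# The resulting material attributes pass through MF_CreativeMoveTool.
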Material /Game/Creative/Devices/Trigger/Materials/M_TriggerDevice_Animating
graph {
 properties {bUseMaterialAttributes=true, pixelDepthOffsetMode=PDOM_Legacy, usedShadingModels=Default Lit}
 nodes [
  n1 = Subtract(A=Time(), B=ScalarParameter() {parameterName=StartingTime})
  n2 = Multiply(A=Sine(None=Multiply(A=@n1) {constB=0.5}) {period=0.5})
  n3 = SetMaterialAttributes!(
   MaterialAttributes
   Base Color=Constant3Vector() {constant=<rgba 0.0885 0.0885 0.0885 0>}
   Metallic
   Specular
   Roughness
   Emissive Color=LinearInterpolate(A=VectorParameter() {parameterName=Glow, defaultValue=<rgba 1 0.4893 0 1>}.RGB, B=Constant3Vector(), Alpha=Add(A=@n2) {constB=0.25})
  ) {attributeSetTypes [
    #1
    #2
    #3
    #4
    #5
   ]}
  MaterialFunctionCall(MA In=@n3) {materialFunction=&MF_CreativeMoveTool}
 ]
}
