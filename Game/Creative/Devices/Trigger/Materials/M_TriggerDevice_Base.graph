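# Base material for the Creative trigger device, Default Lit and built on material attributes.
# A Glow color parameter, blue by default, is multiplied by Glow_Mult to drive emissive.
# Base color, metallic, specular and roughness are exposed on the attribute node but have nothing wired in.
# The attributes then pass through MF_CreativeMoveTool, a shared function that by its name supports the Creative move tool.
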Material /Game/Creative/Devices/Trigger/Materials/M_TriggerDevice_Base
graph {
 properties {bUseMaterialAttributes=true, pixelDepthOffsetMode=PDOM_Legacy, usedShadingModels=Default Lit}
 nodes [
  n1 = SetMaterialAttributes!(
   MaterialAttributes
   Base Color
   Metallic
   Specular
   Roughness
   Emissive Color=Multiply(
    A=VectorParameter() {parameterName=Glow, defaultValue=<rgba 0.0055 0.3952 0.9479 1>}.RGB
    B=ScalarParameter() {defaultValue=4, parameterName=Glow_Mult}
   )
  ) {attributeSetTypes [
    #1
    #2
    #3
    #4
    #5
   ]}
  MaterialFunctionCall(MA In=@n1) {materialFunction=&MF_CreativeMoveTool}
 ]
}
