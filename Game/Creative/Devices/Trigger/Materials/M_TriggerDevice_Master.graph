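# Master material for the Creative Trigger device, a Default Lit material built on material attributes for instanced static meshes with dithered LOD transitions.
# Fixed diffuse, specular mask and normal textures set base color, specular and normal, then pass through the shared Environment_Effects function.
# The UseLavaEdgeGradient static switch optionally adds the Vulcano lava edge gradient.
# MF_CreativeMoveTool processes the switched attributes as the last step.
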Material /Game/Creative/Devices/Trigger/Materials/M_TriggerDevice_Master
graph {
 properties {
  bAutomaticallySetUsageInEditor=false
  bUseMaterialAttributes=true
  bUsedWithInstancedStaticMeshes=true
  ditheredLODTransition=true
  pixelDepthOffsetMode=PDOM_Legacy
  usedShadingModels=Default Lit
 }
 nodes [
  Constant3Vector() {constant=<rgba 0.0312 0.0312 0.0312 0>}
  n1 = MaterialFunctionCall!(Environment FX=MakeMaterialAttributes!(
    BaseColor=TextureSample() {texture=&/T_CreativeTrigger_D}.RGB
    Metallic
    Specular=TextureSample() {samplerType=SAMPLERTYPE_Masks, texture=&/T_CreativeTrigger_S}.RGB
    Roughness
    Anisotropy
    EmissiveColor
    Opacity
    OpacityMask
    Normal=TextureSample() {samplerType=SAMPLERTYPE_Normal, texture=&/T_CreativeTrigger_N}.RGB
    Tangent
    WorldPositionOffset
    SubsurfaceColor
    ClearCoat
    ClearCoatRoughness
    AmbientOcclusion
    Refraction
    CustomizedUVs_0
    CustomizedUVs_1
    CustomizedUVs_2
    CustomizedUVs_3
    CustomizedUVs_4
    CustomizedUVs_5
    CustomizedUVs_6
    CustomizedUVs_7
    PixelDepthOffset
    ShadingModel
    Displacement
   )) {materialFunction=/Game/Packages/Fortress_SharedMaterials/Base_Material_Functions/Environment_Effects}
  MaterialFunctionCall(MA In=StaticSwitchParameter(
    True=MaterialFunctionCall!(In=@n1.Result, UVs, TransitionGradient) {materialFunction=/Game/Environments/Sets/Vulcano/Materials/MF_LavaEdgeGradientSimple}.Result
    False=@n1.Result
   ) {parameterName=UseLavaEdgeGradient}) {materialFunction=&MF_CreativeMoveTool}
 ]
}
