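# Translucent unlit material for the Creative Trigger device's number display, drawing the Score parameter as LCD-style digits.
# Floored powers-of-ten divisions of Score offset UVs into a digit-strip texture, and ceiled UV masks split the face into digit columns.
# It switches from two-digit to three-digit layout when Score exceeds about 100, while red vertex color selects the texture's green channel.
# Win blends in a sine pulse, and an LCD pattern function masks the Color tint.
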Material /Game/Creative/Devices/Trigger/Materials/M_TriggerDevice_Number
graph {
 properties {blendMode=BLEND_Translucent, pixelDepthOffsetMode=PDOM_Legacy, shadingModel=MSM_Unlit, usedShadingModels=Unlit}
 nodes [
  n1 = TextureCoordinate() {uTiling=0.2}
  n2 = ComponentMask(None=@n1) {r=true}
  n3 = ScalarParameter() {defaultValue=721.3254, parameterName=Score}
  n4 = Multiply(A=Floor(None=Multiply(A=@n3) {constB=0.1})) {constB=0.1}
  n5 = ComponentMask(None=@n1) {g=true}
  n6 = TextureSample(UVs=AppendVector(A=Add(A=@n2, B=@n4), B=@n5)) {samplerType=SAMPLERTYPE_Masks, texture=&/T_NumberSet_Large_CreativeTrigger}
  n7 = Add(A=@n2, B=Multiply(A=Floor(None=@n3)) {constB=0.1})
  n8 = TextureSample(UVs=AppendVector(A=Add(A=@n7) {constB=-0.1}, B=@n5)) {samplerType=SAMPLERTYPE_Masks, texture=&T_NumberSet_Large_CreativeTrigger}
  n9 = TextureCoordinate()
  n10 = Ceil(None=ComponentMask(None=Subtract(A=@n9) {constB=0.5}) {r=true})
  n11 = TextureCoordinate() {uTiling=0.3}
  n12 = ScalarParameter() {defaultValue=721.3254, parameterName=Score}
  n13 = Multiply(A=@n12) {constB=0.01}
  n14 = Add(A=ComponentMask(None=@n11) {r=true}, B=Multiply(A=Floor(None=@n13)) {constB=0.1})
  n15 = ComponentMask(None=@n11) {g=true}
  n16 = TextureCoordinate() {uTiling=0.3}
  n17 = Floor(None=Multiply(A=@n12) {constB=0.1})
  n18 = Add(A=ComponentMask(None=@n16) {r=true}, B=Add(A=Multiply(A=@n17) {constB=0.1}) {constB=-0.1})
  n19 = ComponentMask(None=@n16) {g=true}
  n20 = Subtract(A=ComponentMask(None=TextureCoordinate()) {r=true}) {constB=0.333}
  n21 = LinearInterpolate(
   A=TextureSample(UVs=AppendVector(A=@n14, B=@n15)) {samplerType=SAMPLERTYPE_Masks, texture=&T_NumberSet_Large_CreativeTrigger}.R
   B=TextureSample(UVs=AppendVector(A=@n18, B=@n19)) {samplerType=SAMPLERTYPE_Masks, texture=&T_NumberSet_Large_CreativeTrigger}.R
   Alpha=Ceil(None=@n20)
  )
  n22 = TextureCoordinate() {uTiling=0.3}
  n23 = ComponentMask(None=@n22) {r=true}
  n24 = Add(A=Multiply(A=Floor(None=@n12)) {constB=0.1}) {constB=-0.2}
  n25 = TextureSample(UVs=AppendVector(A=Add(A=@n23, B=@n24), B=ComponentMask(None=@n22) {g=true})) {samplerType=SAMPLERTYPE_Masks, texture=&T_NumberSet_Large_CreativeTrigger}
  n26 = Ceil(None=Subtract(A=@n20) {constB=0.333})
  n27 = Subtract(A=Floor(None=ScalarParameter() {defaultValue=721.3254, parameterName=Score})) {constB=100}
  n28 = LinearInterpolate(
   A=LinearInterpolate(A=LinearInterpolate(A=@n6.R, B=@n8.R, Alpha=@n10), B=LinearInterpolate(A=@n21, B=@n25.R, Alpha=@n26), Alpha=Saturate(None=@n27))
   B=TextureSample(UVs=TextureCoordinate() {uTiling=0.2}) {samplerType=SAMPLERTYPE_Masks, texture=&T_NumberSet_Large_CreativeTrigger}.G
   Alpha=VertexColor().R
  )
  n29 = Multiply(A=Time()) {constB=2}
  LinearInterpolate(A=@n28, B=Multiply(A=@n28, B=Sine(None=@n29)), Alpha=ScalarParameter() {parameterName=Win})
  LinearInterpolate(
   B=VectorParameter() {parameterName=Color, defaultValue=<rgba 2.5 0.1638 0.1638 1>}.RGB
   Alpha=MaterialFunctionCall!(
    U Repeat=ScalarParameter() {defaultValue=35, parameterName=LCD_U}
    V Repeat=ScalarParameter() {defaultValue=15, parameterName=LCD_V}
    Dark Value
    Light Value
   ) {materialFunction=/Game/Effects/Fort_Effects/Effects/Atlas/Materials/MF_Atlas_LCD_01}.Result
  )
 ]
}
